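; PCG test graph that assigns a shape grammar definition to incoming data through the Assign_ShapeGrammarDefinition primitive.
; The graph input feeds the primary input of that single subgraph node.
; Graph parameters for size, size attribute, definition asset and asset spawning, on by default, pass straight through to the subgraph.
; Nothing connects to the graph output.
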(graph AssignShapeGrammar_TestGraph
  (params (Size :type Float) (ShapeGrammarDefinition :type (SoftObject /PCGPrimitives/Grammar/BP/BP_ShapeGrammarDefinition.BP_ShapeGrammarDefinition_C)) (UseAttributeForSize :type String) (SpawnAssets :type Bool :default true))
  (bind Subgraph_0 (Subgraph :Graph /PCGPrimitives/Primitives/Assign/Assign_ShapeGrammarDefinition.Assign_ShapeGrammarDefinition :ShapeGrammarDefinition ShapeGrammarDefinition :Size Size :SpawnAssets SpawnAssets :UseAttributeForSize UseAttributeForSize))
  (connect DefaultInputNode In Subgraph_0 PrimaryInput))
